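# Sound cue that plays the down-but-not-out (DBNO) voice line for the VividRazor CoffeePuff character.
# A single wave player feeds the output directly with one recorded take, without randomization.
# It uses the critical enemy VO sound class and a shared NPC "Alerted Ethereal Stereo" attenuation asset for spatial falloff.
# CoffeePuff's Global and Melee VO concurrency groups limit overlapping playback, and the cue's debug flag is left on.
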SoundCue /VividRazor_CoffeePuff/Audio/VO/IP_SC_VividRazor_CoffeePuff_VO_DBNO
root=WavePlayer_5
nodes [{id=WavePlayer_5, properties {soundWaveAssetPtr=&/IP_VividRazor_CoffeePuff_VO_DBNO_01}}]
properties {
 attenuationSettings=/Game/Sounds/Fort_Human_VO_BR/NPCs/Shared/NPC_VO_Alerted_Ethereal_Stereo_Att
 bDebug=true
 maxDistance=12500
 soundClassObject=&Fort_VO_Enemy_Critical
 volumeMultiplier=1
 attenuationOverrides {distanceAlgorithm=Logarithmic}
 concurrencyOverrides {maxCount=3}
 concurrencySet [
  &/IP_CONC_VividRazor_CoffeePuff_VO_Global
  &/IP_CONC_VividRazor_CoffeePuff_VO_Melee
 ]
}
