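# Sound cue for the CoffeePuff character's voice line on the Hellfire Punch slam attack.
# A Random node, excluded from branch culling, picks one of three equally weighted Punch Slam VO waves.
# It plays in the critical enemy VO sound class, uses a shared NPC stereo attenuation asset and joins the character's global VO concurrency group.
# The cue's debug flag is left on.
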SoundCue /VividRazor_CoffeePuff/Audio/VO/IP_SC_VividRazor_CoffeePuff_VO_HellfirePunch_Slam
root=Random_0
nodes [
 {id=Random_0, children [WavePlayer_0,WavePlayer_1,WavePlayer_2], properties {bShouldExcludeFromBranchCulling=true, weights [1,1,1]}}
 {id=WavePlayer_0, properties {soundWaveAssetPtr=&/IP_VividRazor_CoffeePuff_VO_Punch_Slam_01}}
 {id=WavePlayer_1, properties {soundWaveAssetPtr=&/IP_VividRazor_CoffeePuff_VO_Punch_Slam_02}}
 {id=WavePlayer_2, properties {soundWaveAssetPtr=&/IP_VividRazor_CoffeePuff_VO_Punch_Slam_03}}
]
properties {
 attenuationSettings=/Game/Sounds/Fort_Human_VO_BR/NPCs/Shared/NPC_VO_Alerted_Ethereal_Stereo_Att
 bDebug=true
 maxDistance=12500
 soundClassObject=&Fort_VO_Enemy_Critical
 volumeMultiplier=0.9
 attenuationOverrides {distanceAlgorithm=Logarithmic}
 concurrencyOverrides {maxCount=3}
 concurrencySet [&/IP_CONC_VividRazor_CoffeePuff_VO_Global]
}
