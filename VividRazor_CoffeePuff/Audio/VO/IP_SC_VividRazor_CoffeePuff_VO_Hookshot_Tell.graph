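# Sound cue for the VividRazor CoffeePuff enemy's voice line that, by its name, telegraphs a hookshot.
# A Random node, excluded from branch culling, picks one of three equally weighted wave variants.
# It plays on the critical enemy VO sound class with shared NPC alerted-VO attenuation and the character's global VO concurrency set.
# The cue's debug flag is on.
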SoundCue /VividRazor_CoffeePuff/Audio/VO/IP_SC_VividRazor_CoffeePuff_VO_Hookshot_Tell
root=Random_2
nodes [
 {id=Random_2, children [WavePlayer_6,WavePlayer_7,WavePlayer_8], properties {bShouldExcludeFromBranchCulling=true, weights [1,1,1]}}
 {id=WavePlayer_6, properties {soundWaveAssetPtr=&/IP_VividRazor_CoffeePuff_VO_Hookshot_Tell_01}}
 {id=WavePlayer_7, properties {soundWaveAssetPtr=&/IP_VividRazor_CoffeePuff_VO_Hookshot_Tell_02}}
 {id=WavePlayer_8, properties {soundWaveAssetPtr=&/IP_VividRazor_CoffeePuff_VO_Hookshot_Tell_03}}
]
properties {
 attenuationSettings=/Game/Sounds/Fort_Human_VO_BR/NPCs/Shared/NPC_VO_Alerted_Ethereal_Stereo_Att
 bDebug=true
 maxDistance=12500
 soundClassObject=&Fort_VO_Enemy_Critical
 volumeMultiplier=0.794
 attenuationOverrides {distanceAlgorithm=Logarithmic}
 concurrencyOverrides {maxCount=3}
 concurrencySet [&/IP_CONC_VividRazor_CoffeePuff_VO_Global]
}
