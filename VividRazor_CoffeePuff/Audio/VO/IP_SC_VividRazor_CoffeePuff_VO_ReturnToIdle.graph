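# Sound Cue for the VividRazor CoffeePuff voice line on returning to idle, apparently for an NPC, judging by its enemy critical VO sound class.
# The cue has no nodes and a null root, so it references no audio here and holds only playback settings.
# It uses a shared NPC alerted attenuation asset and joins CoffeePuff's global VO and melee concurrency groups.
# Debug is enabled on the cue.
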SoundCue /VividRazor_CoffeePuff/Audio/VO/IP_SC_VividRazor_CoffeePuff_VO_ReturnToIdle
root=null
nodes []
properties {
 attenuationSettings=/Game/Sounds/Fort_Human_VO_BR/NPCs/Shared/NPC_VO_Alerted_Ethereal_Stereo_Att
 bDebug=true
 maxDistance=16000
 soundClassObject=&Fort_VO_Enemy_Critical
 volumeMultiplier=0.794
 attenuationOverrides {distanceAlgorithm=Logarithmic}
 concurrencyOverrides {maxCount=3}
 concurrencySet [
  &/IP_CONC_VividRazor_CoffeePuff_VO_Global
  &/IP_CONC_VividRazor_CoffeePuff_VO_Melee
 ]
}
