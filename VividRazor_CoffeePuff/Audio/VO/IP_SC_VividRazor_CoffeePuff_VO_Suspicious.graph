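# Sound cue for the VividRazor CoffeePuff character's "Suspicious" voice line.
# A random node picks one of four equally weighted voice waves and is excluded from branch culling.
# Playback uses a shared NPC alerted-VO attenuation asset and the critical enemy VO sound class.
# Character-wide and melee VO concurrency sets limit overlapping playback, and the debug flag is left on.
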SoundCue /VividRazor_CoffeePuff/Audio/VO/IP_SC_VividRazor_CoffeePuff_VO_Suspicious
root=Random_2
nodes [
 {
  id=Random_2
  children [WavePlayer_5, WavePlayer_6, WavePlayer_9, WavePlayer_10]
  properties {bShouldExcludeFromBranchCulling=true, weights [1,1,1,1]}
 }
 {id=WavePlayer_10, properties {soundWaveAssetPtr=&/IP_VividRazor_CoffeePuff_VO_Suspicious_04}}
 {id=WavePlayer_5, properties {soundWaveAssetPtr=&/IP_VividRazor_CoffeePuff_VO_Suspicious_01}}
 {id=WavePlayer_6, properties {soundWaveAssetPtr=&/IP_VividRazor_CoffeePuff_VO_Suspicious_02}}
 {id=WavePlayer_9, properties {soundWaveAssetPtr=&/IP_VividRazor_CoffeePuff_VO_Suspicious_03}}
]
properties {
 attenuationSettings=/Game/Sounds/Fort_Human_VO_BR/NPCs/Shared/NPC_VO_Alerted_Ethereal_Stereo_Att
 bDebug=true
 maxDistance=12500
 soundClassObject=&Fort_VO_Enemy_Critical
 volumeMultiplier=0.7
 attenuationOverrides {distanceAlgorithm=Logarithmic}
 concurrencyOverrides {maxCount=3}
 concurrencySet [
  &/IP_CONC_VividRazor_CoffeePuff_VO_Global
  &/IP_CONC_VividRazor_CoffeePuff_VO_Melee
 ]
}
